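; Target actor for the cannonball minigame that plays hit feedback.
; Its HitTarget custom event takes an integer Rank, with 1 or lower counting as a bullseye.
; A bullseye activates the bullseye FX component and plays bullseye and celebration sounds at the target's location, while any other rank plays only a generic hit sound.
; The graph contains no hit detection or scoring.
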(event Custom|HitTarget
  (params (Rank :type Integer))
  (bind _returnvalue (Transformation|GetActorLocation))
  (if (<= Rank 1)
    ;[1 1 1] Hit Bullseye: do fireworks and cheers
    (Components|Activation|Activate (Variables|Default|GetFX_Bullseye) true)
    (Audio|PlaySoundatLocation (Variables|Default|GetSFX_Bullseye) _returnvalue)
    (Audio|PlaySoundatLocation (Variables|Default|GetSFX_Celebration) _returnvalue)
    (else
      ;[1 1 1] Hit Generic: success audio without fanfare
      (Audio|PlaySoundatLocation (Variables|Default|GetSFX_Generic) (Transformation|GetActorLocation)))))

(fn ConstructionScript)
