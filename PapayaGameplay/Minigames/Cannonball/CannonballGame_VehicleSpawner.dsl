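; Vehicle spawner for the Cannonball minigame that supplies a cannon and keeps it tied to a play volume.
; On authority, a data-row hotfix check either runs the parent spawn once or destroys the spawner.
; The cannon's OnDestroyed clears the damage timer and re-runs the parent spawn after a short delay.
; Leaving CannonVolume starts a looping timer that applies a damage GameplayEffect to the cannon, and re-entering clears it.
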(event OnComponentBeginOverlap(CannonVolume)
  (params (OverlappedComponent :type Primitive%20Component%20Object%20Reference) (OtherActor :type Actor%20Object%20Reference) (OtherComp :type Primitive%20Component%20Object%20Reference) (OtherBodyIndex :type Integer) (bFromSweep :type Boolean) (SweepResult :type Hit%20Result%20Structure%20%28by%20ref%29))
  (bind _spawnedvehicle (Variables|Default|GetSpawnedVehicle))
  ;[1 1 1] Begin Overlap -> Clear the damage timer when the cannon enters the volume
  (Utilities|IsValid _spawnedvehicle
    (:Is%20Valid
      (if (Utilities|Equal(Object) _spawnedvehicle OtherActor)
        (Utilities|Time|ClearandInvalidateTimerbyHandle (Variables|Default|GetDamageTimer))))))

(event OnComponentEndOverlap(CannonVolume)
  (params (OverlappedComponent :type Primitive%20Component%20Object%20Reference) (OtherActor :type Actor%20Object%20Reference) (OtherComp :type Primitive%20Component%20Object%20Reference) (OtherBodyIndex :type Integer))
  (bind _spawnedvehicle (Variables|Default|GetSpawnedVehicle))
  ;[1 1 1] End Overlap -> Apply damage at intervals when the cannon leaves the volume
  (if (Variables|Default|GetCannonSpawned)
    (Utilities|IsValid _spawnedvehicle
      (:Is%20Valid
        (if (Utilities|Equal(Object) _spawnedvehicle OtherActor)
          (bind _returnvalue (Utilities|Time|SetTimerbyEvent (delegate Custom|ApplyDamage) (Variables|Default|GetDamageInterval) true))
          (Variables|Default|SetDamageTimer _returnvalue))))))

(event Custom|ApplyDamage
  (bind _spawnedvehicle (Variables|Default|GetSpawnedVehicle))
  (bind _returnvalue (Ability|GetAbilitySystemComponent _spawnedvehicle))
  ;[1 1 1] End Overlap -> Apply damage at intervals when the cannon leaves the volume
  (Utilities|IsValid _returnvalue
    (:Is%20Valid
      (GameplayEffects|ApplyGameplayEffectToSelf :self _returnvalue :GameplayEffectClass (Variables|Default|GetGE_Damage) :Level 0.0))))

(event EventSpawnVehicle
  (bind _spawnedvehicle (Variables|Default|GetSpawnedVehicle))
  (Networking|SwitchHasAuthority
    (:Authority
      ;[1 1 1] Hotfix Enabled Check
      (if (Class|AthenaFunctionLibrary|RowtoBool (Variables|Default|GetEnabledRow))
        (Utilities|FlowControl|DoOnce false
          (:Completed
            (|CallParentFunction)
            (Development|PrintString "Spawn Cannon" false false)
            ;[1 1 1] Wait for the cannon to be destroyed and call respawn
            (Utilities|IsValid _spawnedvehicle
              (:Is%20Valid
                (Game|AssignOnDestroyed _spawnedvehicle (delegate Custom|OnCannonDestroyed))
                (Variables|Default|SetCannonSpawned true)))))
        (else
          (Actor|DestroyActor))))))

(event Custom|RespawnCannon
  (bind _spawnedvehicle (Variables|Default|GetSpawnedVehicle))
  (|CallParentFunction)
  (Development|PrintString "Spawn Cannon" false false)
  ;[1 1 1] Wait for the cannon to be destroyed and call respawn
  (Utilities|IsValid _spawnedvehicle
    (:Is%20Valid
      (Game|AssignOnDestroyed _spawnedvehicle (delegate Custom|OnCannonDestroyed))
      (Variables|Default|SetCannonSpawned true))))

(event Custom|OnCannonDestroyed
  (params (DestroyedActor :type Actor%20Object%20Reference))
  ;[1 1 1] Wait for the cannon to be destroyed and call respawn
  (Variables|Default|SetCannonSpawned false)
  (Utilities|Time|SetTimerbyEvent (EventDispatchers|CreateEvent :!event Custom|RespawnCannon) 1.0)
  (Utilities|Time|ClearandInvalidateTimerbyHandle (Variables|Default|GetDamageTimer))
  (Development|PrintString "Respawn Cannon" false false))

(fn ConstructionScript)
