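; Gameplay ability for a player launched in the cannonball minigame, tracking the flight and scoring the landing on a bullseye target.
; On the locally controlled client, a repeating timer reports distance from launch to the minigame activity component.
; An actor-hit handler ranks bullseye distance, notifies the target and reports the result, guarded by a Completed flag.
; Explosion knockback triggers a delayed end, which reports failure if still unscored.
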(event Ability|EventActivateAbility
  (bind _player (Variables|Default|GetPlayer))
  ;[1 1 1] Begin Play
  (|CallParentFunction)
  (Variables|Default|SetCompleted false)
  (bind _asfort_player_pawn (Utilities|Casting|CastToFortPlayerPawn (Ability|GetAvatarActorfromActorInfo)))
  (bind _output_get (Variables|Default|SetPlayer _asfort_player_pawn))
  (Collision|AssignOnActorHit _output_get (delegate Custom|OnActorHit_Event))
  ;[1 1 1] Minigame Start
  (Ability|HasAuthority
    (:False
      (Ability|IsLocallyControlled
        (:True
          (Audio|PlaySound2D (Variables|Sound|GetStart%20Sound))
          (bind _returnvalue (Utilities|Time|SetTimerbyEvent (EventDispatchers|CreateEvent :!event Custom|CannonballTick) 0.05 true))
          (Variables|Default|SetCannonballTimer _returnvalue)
          (Variables|Default|SetStartingLocation (Transformation|GetActorLocation _player))
          (bind _returnvalue_2 (Class|FortControllerComponentMinigameActivity|GetMinigameActivityComponent _player)
            (:Valid
              (Class|FortControllerComponentMinigameActivity|ForwardMinigameActivityStarted _returnvalue_2 (Utilities|Struct|MakeMinigameActivityStartedData (Variables|Default|GetMinigameActivityName) "Distance")))))))))

(event Custom|OnActorHit_Event
  (params (SelfActor :type Actor%20Object%20Reference) (OtherActor :type Actor%20Object%20Reference) (NormalImpulse :type Vector) (Hit :type Hit%20Result%20Structure%20%28by%20ref%29))
  (bind (_bblockinghit _binitialoverlap _time _distance _location) (Collision|BreakHitResult Hit))
  ;[1 1 1] OnHitActor -> Did we hit the target?
  (bind _asbp_cannonball_game_target (Utilities|Casting|CastToBP_CannonballGame_Target OtherActor))
  (Development|PrintString (Utilities|String|ToString(Float) (Math|Vector|Distance(Vector) _location (Transformation|GetWorldLocation (Variables|/PapayaGameplay/Minigames/Cannonball/BP_CannonballGame_Target.BP_CannonballGame_Target_C|GetBullseye _asbp_cannonball_game_target)))) true true "(R=0.000000,G=0.660000,B=1.000000,A=1.000000)" 1e+01)
  (Class|BPCannonballGameTarget|HitTarget _asbp_cannonball_game_target (CallFunction|ScoreRank (Math|Vector|Distance(Vector) _location (Transformation|GetWorldLocation (Variables|/PapayaGameplay/Minigames/Cannonball/BP_CannonballGame_Target.BP_CannonballGame_Target_C|GetBullseye _asbp_cannonball_game_target)))))
  (CallFunction|UpdateMiniGameScore (Math|Vector|Distance(Vector) _location (Transformation|GetWorldLocation (Variables|/PapayaGameplay/Minigames/Cannonball/BP_CannonballGame_Target.BP_CannonballGame_Target_C|GetBullseye _asbp_cannonball_game_target))) (CallFunction|ScoreRank (Math|Vector|Distance(Vector) _location (Transformation|GetWorldLocation (Variables|/PapayaGameplay/Minigames/Cannonball/BP_CannonballGame_Target.BP_CannonballGame_Target_C|GetBullseye _asbp_cannonball_game_target)))) true)
  (Utilities|Time|ClearandInvalidateTimerbyHandle (Variables|Default|GetCannonballTimer)))

(event Custom|UpdateMiniGameScore
  (params (Distance :type Float%20%28double-precision%29) (Rank :type Integer) (Success :type Boolean))
  ;[1 1 1] Update the Minigame Score on the Client
  (if (not (Variables|Default|GetCompleted))
    (Variables|Default|SetCompleted true)
    (bind _returnvalue (Class|FortControllerComponentMinigameActivity|GetMinigameActivityComponent (Variables|Default|GetPlayer))
      (:Valid
        (Class|FortControllerComponentMinigameActivity|ForwardMinigameActivityEnded _returnvalue (Utilities|Struct|MakeMinigameActivityEndedData Success "Distance" Distance 0.0 -1 -1 Rank))
        (Audio|PlaySound2D (select Success (Variables|Sound|GetSuccessSound) (Variables|Sound|GetFailed%20Sound)))))))

(event EventDoExplosionAndKnockback
  ;[1 1 1] End Ability
  (|CallParentFunction)
  (Ability|Tasks|WaitDelay (Variables|Default|GetEndAbilityDelay)
    (:OnFinish
      (Ability|EndAbility))))

(event Ability|EventOnEndAbility
  (params (bWasCancelled :type Boolean))
  (bind _player (Variables|Default|GetPlayer))
  ;[1 1 1] End Ability
  (|CallParentFunction bWasCancelled)
  (Utilities|Time|ClearandInvalidateTimerbyHandle (Variables|Default|GetCannonballTimer))
  (Utilities|IsValid _player
    (:Is%20Valid
      (Collision|UnbindallEventsfromOnActorHit _player)
      (CallFunction|UpdateMiniGameScore 0.0 -1))))

(event Custom|CannonballTick
  (bind _player (Variables|Default|GetPlayer)
    (:then
      (bind _returnvalue (Class|FortControllerComponentMinigameActivity|GetMinigameActivityComponent _player)
        (:Valid
          (Class|FortControllerComponentMinigameActivity|ForwardMinigameActivityStatChanged _returnvalue "Distance" (Math|Vector|Distance(Vector) (Transformation|GetActorLocation _player) (Variables|Default|GetStartingLocation))))))
    (:else)))

(fn ScoreRank
  (params (Distance :type Float%20%28double-precision%29))
  (if (<= Distance (Variables|Default|GetDistance_Rank0))
    (return 0)
    (elif (<= Distance (Variables|Default|GetDistance_Rank1))
      (return 1)
      (elif (<= Distance (Variables|Default|GetDistance_Rank2))
        (return 2)
        (elif (<= Distance (Variables|Default|GetDistance_Rank3))
          (return 3)
          (else
            (return 4)))))))
